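# Looping ambient sound cue for an extra-large waterfall, routed to the water ambience sound class at reduced volume with a waterfall-specific attenuation asset.
# A quality-level root picks one of four pitch-modulated branches: a random choice between two waterfall loops, a single loop, or two branches using an optimized wave.
# The first two branches are pitched up, while the optimized ones vary around normal pitch.
# A concurrency override caps instances, stopping the quietest.
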SoundCue /Figment_S05/Sounds/Fort_Ambient_Looping/Water/Waterfall_ExtraLarge_A_01_Cue
root=QualityLevel_0
nodes [
 {id=Modulator_0, children [WavePlayer_4], properties {pitchMax=1.3, pitchMin=1.2}}
 {id=Modulator_1, children [WavePlayer_1], properties {pitchMax=1.1, pitchMin=0.9}}
 {id=Modulator_2, children [Random_0], properties {pitchMax=1.3, pitchMin=1.2}}
 {id=Modulator_3, children [WavePlayer_2], properties {pitchMax=1.1, pitchMin=0.9}}
 {id=QualityLevel_0, children [Modulator_2,Modulator_0,Modulator_1,Modulator_3]}
 {id=Random_0, children [WavePlayer_0,WavePlayer_3], properties {weights [1,1]}}
 {id=WavePlayer_0, properties {bLooping=true, soundWaveAssetPtr=/Game/Sounds/Fort_Ambient_Looping/Water/waterfall_new_ST_01}}
 {id=WavePlayer_1, properties {bLooping=true, soundWaveAssetPtr=/Game/Sounds/Fort_Ambient_Looping/Water/waterfall_new_ST_01_opt}}
 {id=WavePlayer_2, properties {bLooping=true, soundWaveAssetPtr=&waterfall_new_ST_01_opt}}
 {id=WavePlayer_3, properties {bLooping=true, soundWaveAssetPtr=/Game/Sounds/Fort_Ambient_Looping/Water/waterfall_new_ST_03}}
 {id=WavePlayer_4, properties {bLooping=true, soundWaveAssetPtr=&waterfall_new_ST_01}}
]
properties {
 attenuationSettings=/Figment_S05/Sounds/Attenuation/Fort_Ambient_Capsule_Waterfall_ExtraLarge_01_Att
 bDebug=true
 bOverrideConcurrency=true
 maxDistance=5000
 soundClassObject=/Game/Sounds/FortSoundClassesAndModes/Fort_Ambient_Water
 volumeMultiplier=0.5
 concurrencyOverrides {maxCount=4, resolutionRule=StopQuietest}
}
